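# Control Rig for the Missing Link male medium body cosmetic, driving helper bones for its bun and noodle arms on top of the body pose.
# The top and bottom center bun bones aim at each other, while clavicle and shoulder helpers follow the upper arms through aims, offsets and constraints.
# Each arm fits a stretched noodle chain to a four-point Bezier from shoulder to hand, with offsets blended by target_alpha.
# LOD2 weights resetting twist bones to follow their parent rigidly.
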ControlRigBlueprint /BRCosmetics/Characters/Player/Male/Medium/Bodies/M_MED_Missing_Link/Meshes/M_MED_Missing_Link_CtrlRig
graphs [
 RigVMFunctionLibrary {links [], nodes []}
 RigVMModel {nodes [
   n1 = &RigUnit_BeginExecution()
   SequenceAggregate = RigVMAggregateNode(ExecuteContext=@n1.ExecutePin)
   /Script/ControlRig.RigUnit_DebugTransformMutable(
    Color=(R=1,G=0,B=0,A=1)
    DebugDrawSettings=(DepthPriority=SDPG_Foreground,Lifetime=-1)
    Mode=Axes
    Scale=10
    Space=None
    Thickness=0
    WorldOffset=(Rotation=(X=0,Y=0,Z=0,W=1),Translation=(X=0,Y=0,Z=0),Scale3D=(X=1,Y=1,Z=1))
    bEnabled=True
    Transform=&RigUnit_GetBoneTransform(Bone=bunBottom_c).Transform
    ExecutePin=@SequenceAggregate.A
   ) {is=struct RigVMUnitNode}
   > /Script/ControlRig.RigUnit_AimBone(
    Bone=bunTop_c
    DebugSettings=(bEnabled=true,Scale=10,WorldOffset=(Rotation=(X=0,Y=0,Z=0,W=1),Translation=(X=0,Y=0,Z=0),Scale3D=(X=1,Y=1,Z=1)))
    Primary=(Weight=1,Axis=(X=-1,Y=0,Z=0),Target=(X=0,Y=0,Z=0),Kind=Location,Space="bunBottom_c")
    Secondary=(Weight=0,Axis=(X=0,Y=0,Z=1),Target=(X=0,Y=0,Z=0),Kind=Direction,Space="None")
    Weight=1
    bPropagateToChildren=True
   ) {is=struct RigVMUnitNode}
   > &RigUnit_AimBone(
    Bone=bunBottom_c
    DebugSettings=(bEnabled=true,Scale=10,WorldOffset=(Rotation=(X=0,Y=0,Z=0,W=1),Translation=(X=0,Y=0,Z=0),Scale3D=(X=1,Y=1,Z=1)))
    Primary=(Weight=1,Axis=(X=1,Y=0,Z=0),Target=(X=0,Y=0,Z=0),Kind=Location,Space="bunTop_c")
    Secondary=(Weight=0,Axis=(X=0,Y=0,Z=1),Target=(X=0,Y=0,Z=0),Kind=Direction,Space="None")
    Weight=1
    bPropagateToChildren=True
   )
   n2 = &RigVMFunction_MathTransformMul(
    B=&RigUnit_GetBoneTransform(Bone=upperarm_l).Transform
    A=/Script/RigVM.RigVMFunction_MathTransformInverse(Value=/Script/ControlRig.RigUnit_GetInitialBoneTransform(Bone=upperarm_l, Space=GlobalSpace) {is=struct RigVMUnitNode}.Transform) {is=struct RigVMUnitNode, template="Inverse::Execute(in Value,out Result)"}.Result
   ) {template="Multiply::Execute(in A,in B,out Result)"}
   n3 = RigVMVariableNode() {variable=LOD2}
   /Script/ControlRig.RigUnit_ModifyBoneTransforms(
    BoneToModify=((Bone="upperarm_twist_01_l",Transform=(Rotation=(X=0,Y=0,Z=0,W=1),Translation=(X=0,Y=0,Z=0),Scale3D=(X=1,Y=1,Z=1))))
    Mode=OverrideGlobal
    WeightMaximum=1
    WeightMinimum=0
    BoneToModify.0.Transform=&RigVMFunction_MathTransformMul(A=&RigUnit_GetInitialBoneTransform(Bone=upperarm_twist_01_l, Space=GlobalSpace).Transform, B=@n2.Result) {template="Multiply::Execute(in A,in B,out Result)"}.Result
    ExecutePin=@SequenceAggregate.B
    Weight=@n3.Value
   ) {is=struct RigVMUnitNode}
   > &RigUnit_ModifyBoneTransforms(
    BoneToModify=((Bone="upperarm_twist_02_l",Transform=(Rotation=(X=0,Y=0,Z=0,W=1),Translation=(X=0,Y=0,Z=0),Scale3D=(X=1,Y=1,Z=1))))
    Mode=OverrideGlobal
    WeightMaximum=1
    WeightMinimum=0
    BoneToModify.0.Transform=&RigVMFunction_MathTransformMul(A=&RigUnit_GetInitialBoneTransform(Bone=upperarm_twist_02_l, Space=GlobalSpace).Transform, B=@n2.Result) {template="Multiply::Execute(in A,in B,out Result)"}.Result
    Weight=@n3.Value
   )
   n4 = &RigVMFunction_MathTransformMul(
    B=&RigUnit_GetBoneTransform(Bone=lowerarm_l).Transform
    A=&RigVMFunction_MathTransformInverse(Value=&RigUnit_GetInitialBoneTransform(Bone=lowerarm_l, Space=GlobalSpace).Transform) {template="Inverse::Execute(in Value,out Result)"}.Result
   ) {template="Multiply::Execute(in A,in B,out Result)"}
   > &RigUnit_ModifyBoneTransforms(
    BoneToModify=((Bone="lowerarm_twist_01_l",Transform=(Rotation=(X=0,Y=0,Z=0,W=1),Translation=(X=0,Y=0,Z=0),Scale3D=(X=1,Y=1,Z=1))))
    Mode=OverrideGlobal
    WeightMaximum=1
    WeightMinimum=0
    BoneToModify.0.Transform=&RigVMFunction_MathTransformMul(A=&RigUnit_GetInitialBoneTransform(Bone=lowerarm_twist_01_l, Space=GlobalSpace).Transform, B=@n4.Result) {template="Multiply::Execute(in A,in B,out Result)"}.Result
    Weight=@n3.Value
   )
   > &RigUnit_ModifyBoneTransforms(
    BoneToModify=((Bone="lowerarm_twist_02_l",Transform=(Rotation=(X=0,Y=0,Z=0,W=1),Translation=(X=0,Y=0,Z=0),Scale3D=(X=1,Y=1,Z=1))))
    Mode=OverrideGlobal
    WeightMaximum=1
    WeightMinimum=0
    BoneToModify.0.Transform=&RigVMFunction_MathTransformMul(A=&RigUnit_GetInitialBoneTransform(Bone=lowerarm_twist_02_l, Space=GlobalSpace).Transform, B=@n4.Result) {template="Multiply::Execute(in A,in B,out Result)"}.Result
    Weight=@n3.Value
   )
   n5 = &RigUnit_GetBoneTransform(Bone=clavicle_l, Space=LocalSpace)
   n6 = &RigVMFunction_MathQuaternionToEuler(RotationOrder=ZYX, Value=@n5.Transform.Rotation)
   > /Script/ControlRig.RigUnit_SetBoneTransform(
    Bone=clavicle_l
    Space=LocalSpace
    Transform=(Rotation=(X=0,Y=0,Z=0,W=1),Translation=(X=0,Y=0,Z=0),Scale3D=(X=1,Y=1,Z=1))
    Weight=1
    bPropagateToChildren=True
    Transform.Translation=@n5.Transform.Translation
    Transform.Rotation=&RigVMFunction_MathQuaternionFromEuler(Euler=/Script/RigVM.RigVMFunction_MathVectorAdd(B=(X=0,Y=0,Z=0), A=@n6.Result) {is=struct RigVMUnitNode, template="Add::Execute(in A,in B,out Result)"}.Result).Result
   ) {is=struct RigVMUnitNode}
   > &RigUnit_AimBone(
    Bone=armDirection_l
    DebugSettings=(bEnabled=False,Scale=10,WorldOffset=(Rotation=(X=0,Y=0,Z=0,W=1),Translation=(X=0,Y=0,Z=0),Scale3D=(X=1,Y=1,Z=1)))
    Primary=(Weight=1,Axis=(X=-1,Y=0,Z=0),Target=(X=0,Y=0,Z=0),Kind=Location,Space="hand_l")
    Secondary=(Weight=0,Axis=(X=0,Y=0,Z=1),Target=(X=1,Y=0,Z=0),Kind=Direction,Space="None")
    Weight=1
    bPropagateToChildren=False
   )
   > &RigUnit_DebugTransformMutable(
    Color=(R=1,G=0,B=0,A=1)
    DebugDrawSettings=(DepthPriority=SDPG_Foreground,Lifetime=-1)
    Mode=Axes
    Scale=20
    Space=armDirection_l
    Thickness=0
    Transform=(Rotation=(X=0,Y=0,Z=0,W=1),Translation=(X=0,Y=0,Z=0),Scale3D=(X=1,Y=1,Z=1))
    WorldOffset=(Rotation=(X=0,Y=0,Z=0,W=1),Translation=(X=0,Y=0,Z=0),Scale3D=(X=1,Y=1,Z=1))
    bEnabled=True
   )
   > &RigUnit_DebugTransformMutable(
    Color=(R=1,G=0,B=0,A=1)
    DebugDrawSettings=(DepthPriority=SDPG_Foreground,Lifetime=-1)
    Mode=Axes
    Scale=20
    Space=spine_05
    Thickness=0
    Transform=(Rotation=(X=0,Y=0,Z=0,W=1),Translation=(X=0,Y=0,Z=0),Scale3D=(X=1,Y=1,Z=1))
    WorldOffset=(Rotation=(X=0,Y=0,Z=0,W=1),Translation=(X=0,Y=0,Z=0),Scale3D=(X=1,Y=1,Z=1))
    bEnabled=True
   )
   n7 = &RigUnit_GetBoneTransform(Bone=spine_05)
   n8 = &RigUnit_GetRelativeBoneTransform(Bone=armBezier_2_l, Space=armDirection_l)
   n9 = RigVMVariableNode() {variable=target_alpha}
   n10 = /Script/RigVM.RigVMFunction_MathVectorLerp(A=(X=0,Y=0,Z=0), B=(X=0,Y=0,Z=0), T=@n9.Value) {is=struct RigVMUnitNode, template="Interpolate::Execute(in A,in B,in T,out Result)"}
   n11 = &RigVMFunction_MathTransformMakeRelative(
    Global=(Rotation=(X=0,Y=0,Z=0,W=1),Translation=(X=0,Y=0,Z=0),Scale3D=(X=1,Y=1,Z=1))
    Parent=@n7.Transform
    Global.Translation=&RigVMFunction_MathTransformMul(
     A=(Rotation=(X=0,Y=0,Z=0,W=1),Translation=(X=0,Y=0,Z=0),Scale3D=(X=1,Y=1,Z=1))
     B=&RigUnit_GetBoneTransform(Bone=armDirection_l).Transform
     A.Translation=&RigVMFunction_MathVectorAdd(A=@n8.Transform.Translation, B=@n10.Result) {template="Add::Execute(in A,in B,out Result)"}.Result
    ) {template="Multiply::Execute(in A,in B,out Result)"}.Result.Translation
   )
   n12 = &RigVMFunction_MathVectorLerp(A=(X=0,Y=0,Z=0), B=(X=0,Y=0,Z=0), T=@n9.Value) {template="Interpolate::Execute(in A,in B,in T,out Result)"}
   n13 = &RigVMFunction_MathTransformMakeRelative(Global=&RigUnit_GetBoneTransform(Bone=shoulder_l).Transform, Parent=@n7.Transform)
   n14 = &RigVMFunction_MathVectorLerp(A=(X=0,Y=0,Z=0), B=(X=0,Y=-5,Z=-4), T=@n9.Value) {template="Interpolate::Execute(in A,in B,in T,out Result)"}
   n15 = /Script/RigVM.RigVMFunction_MathVectorMakeBezierFourPoint(
    Bezier=(A=(X=0,Y=0,Z=0),B=(X=0,Y=0,Z=0),C=(X=0,Y=0,Z=0),D=(X=0,Y=0,Z=0))
    Bezier.D=&RigUnit_GetBoneTransform(Bone=hand_l).Transform.Translation
    Bezier.C=&RigUnit_GetBoneTransform(Bone=armBezier_3_l).Transform.Translation
    Bezier.B=&RigVMFunction_MathTransformMul(
     A=(Rotation=(X=0,Y=0,Z=0,W=1),Translation=(X=0,Y=0,Z=0),Scale3D=(X=1,Y=1,Z=1))
     B=@n7.Transform
     A.Translation=&RigVMFunction_MathVectorAdd(A=@n11.Local.Translation, B=@n12.Result) {template="Add::Execute(in A,in B,out Result)"}.Result
    ) {template="Multiply::Execute(in A,in B,out Result)"}.Result.Translation
    Bezier.A=&RigVMFunction_MathTransformMul(
     A=(Rotation=(X=0,Y=0,Z=0,W=1),Translation=(X=0,Y=0,Z=0),Scale3D=(X=1,Y=1,Z=1))
     B=@n7.Transform
     A.Translation=&RigVMFunction_MathVectorAdd(A=@n13.Local.Translation, B=@n14.Result) {template="Add::Execute(in A,in B,out Result)"}.Result
    ) {template="Multiply::Execute(in A,in B,out Result)"}.Result.Translation
   ) {is=struct RigVMUnitNode}
   > /Script/ControlRig.RigUnit_DebugBezier(
    Color=(R=1,G=0,B=0,A=1)
    DebugDrawSettings=(DepthPriority=SDPG_Foreground,Lifetime=-1)
    Detail=16
    MaximumU=1
    MinimumU=0
    Space=None
    Thickness=0
    WorldOffset=(Rotation=(X=0,Y=0,Z=0,W=1),Translation=(X=0,Y=0,Z=0),Scale3D=(X=1,Y=1,Z=1))
    bEnabled=True
    Bezier=@n15.Bezier
   ) {is=struct RigVMUnitNode}
   > /Script/ControlRig.RigUnit_FitChainToCurve(
    Alignment=Stretched
    DebugSettings=(bEnabled=true,Scale=1,CurveColor=(R=1,G=1,B=0,A=1),SegmentsColor=(R=1,G=0,B=0,A=1),WorldOffset=(Rotation=(X=0,Y=0,Z=0,W=1),Translation=(X=0,Y=0,Z=0),Scale3D=(X=1,Y=1,Z=1)))
    EndBone=noodle_7_l
    Maximum=1
    Minimum=0
    PoleVectorPosition=(X=0,Y=0,Z=0)
    PrimaryAxis=(X=-1,Y=0,Z=0)
    RotationEaseType=Linear
    Rotations=()
    SamplingPrecision=12
    SecondaryAxis=(X=0,Y=0,Z=0)
    StartBone=noodle_1_l
    Weight=1
    bPropagateToChildren=True
    Bezier=@n15.Bezier
   ) {is=struct RigVMUnitNode}
   > &RigUnit_SetTransform(
    Item=(Type=Bone,Name="upperarm_follow_l_zero")
    Weight=1
    bInitial=true
    bPropagateToChildren=false
    Value=&RigUnit_GetBoneTransform(Bone=spine_05).Transform
   )
   n16 = &RigUnit_GetRelativeBoneTransform(Bone=upperarm_l, Space=upperarm_follow_l)
   > /Script/ControlRig.RigUnit_SetBoneTranslation(
    Bone=upperarm_follow_l
    Space=LocalSpace
    Translation=(X=0,Y=0,Z=0)
    Weight=1
    bPropagateToChildren=True
    Translation.X=&RigUnit_GetBoneTransform(Bone=upperarm_follow_l, Space=LocalSpace).Transform.Translation.X
    Translation.Y=@n16.Transform.Translation.Y
    Translation.Z=@n16.Transform.Translation.Z
   ) {is=struct RigVMUnitNode}
   > &RigUnit_AimBone(
    Bone=clavicle_aim_l
    DebugSettings=(bEnabled=False,Scale=10,WorldOffset=(Rotation=(X=0,Y=0,Z=0,W=1),Translation=(X=0,Y=0,Z=0),Scale3D=(X=1,Y=1,Z=1)))
    Primary=(Weight=1,Axis=(X=0,Y=0,Z=-1),Target=(X=0,Y=0,Z=0),Kind=Location,Space="clavicle_tgt_l")
    Secondary=(Weight=1,Axis=(X=-1,Y=0,Z=0),Target=(X=0,Y=0,Z=0),Kind=Location,Space="clavicle_l")
    Weight=1
    bPropagateToChildren=True
   )
   > /Script/ControlRig.RigUnit_TransformConstraint(
    BaseBone=None
    BaseTransform=(Rotation=(X=0,Y=0,Z=0,W=1),Translation=(X=0,Y=0,Z=0),Scale3D=(X=1,Y=1,Z=1))
    BaseTransformSpace=GlobalSpace
    Bone=shoulder_l_zero
    Targets=((Transform=(Rotation=(X=0,Y=0,Z=0,W=1),Translation=(X=0,Y=0,Z=0),Scale3D=(X=1,Y=1,Z=1)),Weight=1,bMaintainOffset=True,Filter=(TranslationFilter=(bX=True,bY=True,bZ=True),RotationFilter=(bX=True,bY=True,bZ=True),ScaleFilter=(bX=True,bY=True,bZ=True))))
    bUseInitialTransforms=True
    Targets.0.Transform=&RigUnit_GetBoneTransform(Bone=clavicle_aim_l).Transform
   ) {is=struct RigVMUnitNode}
   n17 = &RigUnit_GetRelativeBoneTransform(Bone=shoulderOffset_tgt_l, Space=shoulderOffset_l_zero)
   > &RigUnit_SetBoneTranslation(
    Bone=shoulderOffset_l
    Space=LocalSpace
    Translation=(X=0,Y=0,Z=0)
    Weight=1
    bPropagateToChildren=True
    Translation.X=&RigUnit_GetBoneTransform(Bone=shoulderOffset_l, Space=LocalSpace).Transform.Translation.X
    Translation.Y=@n17.Transform.Translation.Y
    Translation.Z=@n17.Transform.Translation.Z
   )
   > /Script/ControlRig.RigUnit_SetBoneRotation(
    Bone=shoulder_l
    Space=GlobalSpace
    Weight=1
    bPropagateToChildren=True
    Rotation=&RigUnit_GetBoneTransform(Bone=upperarm_l).Transform.Rotation
   ) {is=struct RigVMUnitNode}
   > &RigUnit_AimBone(
    Bone=shoulder_aim_l
    DebugSettings=(bEnabled=False,Scale=10,WorldOffset=(Rotation=(X=0,Y=0,Z=0,W=1),Translation=(X=0,Y=0,Z=0),Scale3D=(X=1,Y=1,Z=1)))
    Primary=(Weight=1,Axis=(X=1,Y=0,Z=0),Target=(X=0,Y=0,Z=0),Kind=Location,Space="lowerarm_l")
    Secondary=(Weight=0,Axis=(X=1,Y=0,Z=0),Target=(X=0,Y=0,Z=0),Kind=Location,Space="None")
    Weight=1
    bPropagateToChildren=True
   )
   > &RigUnit_SetBoneRotation(
    Bone=shoulder_twist_01_l
    Space=LocalSpace
    Weight=1
    bPropagateToChildren=False
    Rotation=&RigUnit_GetBoneTransform(Bone=upperarm_twist_01_l, Space=LocalSpace).Transform.Rotation
   )
   n18 = &RigUnit_BeginExecution()
   n19 = &RigVMFunction_MathTransformMul(
    B=&RigUnit_GetBoneTransform(Bone=upperarm_r).Transform
    A=&RigVMFunction_MathTransformInverse(Value=&RigUnit_GetInitialBoneTransform(Bone=upperarm_r, Space=GlobalSpace).Transform) {template="Inverse::Execute(in Value,out Result)"}.Result
   ) {template="Multiply::Execute(in A,in B,out Result)"}
   &RigUnit_ModifyBoneTransforms(
    BoneToModify=((Bone="upperarm_twist_01_r",Transform=(Rotation=(X=0,Y=0,Z=0,W=1),Translation=(X=0,Y=0,Z=0),Scale3D=(X=1,Y=1,Z=1))))
    Mode=OverrideGlobal
    WeightMaximum=1
    WeightMinimum=0
    ExecutePin=@n18.ExecutePin
    BoneToModify.0.Transform=&RigVMFunction_MathTransformMul(A=&RigUnit_GetInitialBoneTransform(Bone=upperarm_twist_01_r, Space=GlobalSpace).Transform, B=@n19.Result) {template="Multiply::Execute(in A,in B,out Result)"}.Result
    Weight=@n3.Value
   )
   > &RigUnit_ModifyBoneTransforms(
    BoneToModify=((Bone="upperarm_twist_02_r",Transform=(Rotation=(X=0,Y=0,Z=0,W=1),Translation=(X=0,Y=0,Z=0),Scale3D=(X=1,Y=1,Z=1))))
    Mode=OverrideGlobal
    WeightMaximum=1
    WeightMinimum=0
    BoneToModify.0.Transform=&RigVMFunction_MathTransformMul(A=&RigUnit_GetInitialBoneTransform(Bone=upperarm_twist_02_r, Space=GlobalSpace).Transform, B=@n19.Result) {template="Multiply::Execute(in A,in B,out Result)"}.Result
    Weight=@n3.Value
   )
   n20 = &RigVMFunction_MathTransformMul(
    B=&RigUnit_GetBoneTransform(Bone=lowerarm_r).Transform
    A=&RigVMFunction_MathTransformInverse(Value=&RigUnit_GetInitialBoneTransform(Bone=lowerarm_r, Space=GlobalSpace).Transform) {template="Inverse::Execute(in Value,out Result)"}.Result
   ) {template="Multiply::Execute(in A,in B,out Result)"}
   > &RigUnit_ModifyBoneTransforms(
    BoneToModify=((Bone="lowerarm_twist_01_r",Transform=(Rotation=(X=0,Y=0,Z=0,W=1),Translation=(X=0,Y=0,Z=0),Scale3D=(X=1,Y=1,Z=1))))
    Mode=OverrideGlobal
    WeightMaximum=1
    WeightMinimum=0
    BoneToModify.0.Transform=&RigVMFunction_MathTransformMul(A=&RigUnit_GetInitialBoneTransform(Bone=lowerarm_twist_01_r, Space=GlobalSpace).Transform, B=@n20.Result) {template="Multiply::Execute(in A,in B,out Result)"}.Result
    Weight=@n3.Value
   )
   > &RigUnit_ModifyBoneTransforms(
    BoneToModify=((Bone="lowerarm_twist_02_r",Transform=(Rotation=(X=0,Y=0,Z=0,W=1),Translation=(X=0,Y=0,Z=0),Scale3D=(X=1,Y=1,Z=1))))
    Mode=OverrideGlobal
    WeightMaximum=1
    WeightMinimum=0
    BoneToModify.0.Transform=&RigVMFunction_MathTransformMul(A=&RigUnit_GetInitialBoneTransform(Bone=lowerarm_twist_02_r, Space=GlobalSpace).Transform, B=@n20.Result) {template="Multiply::Execute(in A,in B,out Result)"}.Result
    Weight=@n3.Value
   )
   n21 = &RigUnit_GetBoneTransform(Bone=clavicle_r, Space=LocalSpace)
   n22 = &RigVMFunction_MathQuaternionToEuler(RotationOrder=ZYX, Value=@n21.Transform.Rotation)
   > &RigUnit_SetBoneTransform(
    Bone=clavicle_r
    Space=LocalSpace
    Transform=(Rotation=(X=0,Y=0,Z=0,W=1),Translation=(X=0,Y=0,Z=0),Scale3D=(X=1,Y=1,Z=1))
    Weight=1
    bPropagateToChildren=True
    Transform.Translation=@n21.Transform.Translation
    Transform.Rotation=&RigVMFunction_MathQuaternionFromEuler(Euler=&RigVMFunction_MathVectorAdd(B=(X=0,Y=0,Z=0), A=@n22.Result) {template="Add::Execute(in A,in B,out Result)"}.Result).Result
   )
   > &RigUnit_AimBone(
    Bone=armDirection_r
    DebugSettings=(bEnabled=False,Scale=10,WorldOffset=(Rotation=(X=0,Y=0,Z=0,W=1),Translation=(X=0,Y=0,Z=0),Scale3D=(X=1,Y=1,Z=1)))
    Primary=(Weight=1,Axis=(X=-1,Y=0,Z=0),Target=(X=0,Y=0,Z=0),Kind=Location,Space="hand_r")
    Secondary=(Weight=0,Axis=(X=0,Y=0,Z=1),Target=(X=1,Y=0,Z=0),Kind=Direction,Space="None")
    Weight=1
    bPropagateToChildren=False
   )
   > &RigUnit_DebugTransformMutable(
    Color=(R=1,G=0,B=0,A=1)
    DebugDrawSettings=(DepthPriority=SDPG_Foreground,Lifetime=-1)
    Mode=Axes
    Scale=20
    Space=armDirection_r
    Thickness=0
    Transform=(Rotation=(X=0,Y=0,Z=0,W=1),Translation=(X=0,Y=0,Z=0),Scale3D=(X=1,Y=1,Z=1))
    WorldOffset=(Rotation=(X=0,Y=0,Z=0,W=1),Translation=(X=0,Y=0,Z=0),Scale3D=(X=1,Y=1,Z=1))
    bEnabled=True
   )
   n23 = &RigUnit_GetBoneTransform(Bone=spine_05)
   n24 = &RigUnit_GetRelativeBoneTransform(Bone=armBezier_2_r, Space=armDirection_r)
   n25 = &RigVMFunction_MathVectorLerp(A=(X=0,Y=0,Z=0), B=(X=0,Y=0,Z=0), T=@n9.Value) {template="Interpolate::Execute(in A,in B,in T,out Result)"}
   n26 = &RigVMFunction_MathTransformMakeRelative(
    Global=(Rotation=(X=0,Y=0,Z=0,W=1),Translation=(X=0,Y=0,Z=0),Scale3D=(X=1,Y=1,Z=1))
    Parent=@n23.Transform
    Global.Translation=&RigVMFunction_MathTransformMul(
     A=(Rotation=(X=0,Y=0,Z=0,W=1),Translation=(X=0,Y=0,Z=0),Scale3D=(X=1,Y=1,Z=1))
     B=&RigUnit_GetBoneTransform(Bone=armDirection_r).Transform
     A.Translation=&RigVMFunction_MathVectorAdd(A=@n24.Transform.Translation, B=@n25.Result) {template="Add::Execute(in A,in B,out Result)"}.Result
    ) {template="Multiply::Execute(in A,in B,out Result)"}.Result.Translation
   )
   n27 = &RigVMFunction_MathVectorLerp(A=(X=0,Y=0,Z=0), B=(X=0,Y=0,Z=6), T=@n9.Value) {template="Interpolate::Execute(in A,in B,in T,out Result)"}
   n28 = &RigVMFunction_MathTransformMakeRelative(Parent=@n23.Transform, Global=&RigUnit_GetBoneTransform(Bone=shoulder_r).Transform)
   n29 = &RigVMFunction_MathVectorLerp(A=(X=0,Y=0,Z=0), B=(X=0,Y=-5,Z=4), T=@n9.Value) {template="Interpolate::Execute(in A,in B,in T,out Result)"}
   n30 = &RigVMFunction_MathVectorMakeBezierFourPoint(
    Bezier=(A=(X=0,Y=0,Z=0),B=(X=0,Y=0,Z=0),C=(X=0,Y=0,Z=0),D=(X=0,Y=0,Z=0))
    Bezier.C=&RigUnit_GetBoneTransform(Bone=armBezier_3_r).Transform.Translation
    Bezier.D=&RigUnit_GetBoneTransform(Bone=hand_r).Transform.Translation
    Bezier.B=&RigVMFunction_MathTransformMul(
     A=(Rotation=(X=0,Y=0,Z=0,W=1),Translation=(X=0,Y=0,Z=0),Scale3D=(X=1,Y=1,Z=1))
     B=@n23.Transform
     A.Translation=&RigVMFunction_MathVectorAdd(A=@n26.Local.Translation, B=@n27.Result) {template="Add::Execute(in A,in B,out Result)"}.Result
    ) {template="Multiply::Execute(in A,in B,out Result)"}.Result.Translation
    Bezier.A=&RigVMFunction_MathTransformMul(
     A=(Rotation=(X=0,Y=0,Z=0,W=1),Translation=(X=0,Y=0,Z=0),Scale3D=(X=1,Y=1,Z=1))
     B=@n23.Transform
     A.Translation=&RigVMFunction_MathVectorAdd(A=@n28.Local.Translation, B=@n29.Result) {template="Add::Execute(in A,in B,out Result)"}.Result
    ) {template="Multiply::Execute(in A,in B,out Result)"}.Result.Translation
   )
   > &RigUnit_DebugBezier(
    Color=(R=1,G=0,B=0,A=1)
    DebugDrawSettings=(DepthPriority=SDPG_Foreground,Lifetime=-1)
    Detail=16
    MaximumU=1
    MinimumU=0
    Space=None
    Thickness=0
    WorldOffset=(Rotation=(X=0,Y=0,Z=0,W=1),Translation=(X=0,Y=0,Z=0),Scale3D=(X=1,Y=1,Z=1))
    bEnabled=True
    Bezier=@n30.Bezier
   )
   > &RigUnit_FitChainToCurve(
    Alignment=Stretched
    DebugSettings=(bEnabled=true,Scale=1,CurveColor=(R=1,G=1,B=0,A=1),SegmentsColor=(R=1,G=0,B=0,A=1),WorldOffset=(Rotation=(X=0,Y=0,Z=0,W=1),Translation=(X=0,Y=0,Z=0),Scale3D=(X=1,Y=1,Z=1)))
    EndBone=noodle_7_r
    Maximum=1
    Minimum=0
    PoleVectorPosition=(X=0,Y=0,Z=0)
    PrimaryAxis=(X=-1,Y=0,Z=0)
    RotationEaseType=Linear
    Rotations=()
    SamplingPrecision=12
    SecondaryAxis=(X=0,Y=0,Z=0)
    StartBone=noodle_1_r
    Weight=1
    bPropagateToChildren=True
    Bezier=@n30.Bezier
   )
   > &RigUnit_SetTransform(
    Item=(Type=Bone,Name="upperarm_follow_r_zero")
    Weight=1
    bInitial=true
    bPropagateToChildren=false
    Value=&RigUnit_GetBoneTransform(Bone=spine_05).Transform
   )
   n31 = &RigUnit_GetRelativeBoneTransform(Bone=upperarm_r, Space=upperarm_follow_r)
   > &RigUnit_SetBoneTranslation(
    Bone=upperarm_follow_r
    Space=LocalSpace
    Translation=(X=0,Y=0,Z=0)
    Weight=1
    bPropagateToChildren=True
    Translation.X=&RigUnit_GetBoneTransform(Bone=upperarm_follow_r, Space=LocalSpace).Transform.Translation.X
    Translation.Y=@n31.Transform.Translation.Y
    Translation.Z=@n31.Transform.Translation.Z
   )
   > &RigUnit_AimBone(
    Bone=clavicle_aim_r
    DebugSettings=(bEnabled=False,Scale=10,WorldOffset=(Rotation=(X=0,Y=0,Z=0,W=1),Translation=(X=0,Y=0,Z=0),Scale3D=(X=1,Y=1,Z=1)))
    Primary=(Weight=1,Axis=(X=0,Y=0,Z=1),Target=(X=0,Y=0,Z=0),Kind=Location,Space="clavicle_tgt_r")
    Secondary=(Weight=1,Axis=(X=-1,Y=0,Z=0),Target=(X=0,Y=0,Z=0),Kind=Location,Space="clavicle_r")
    Weight=1
    bPropagateToChildren=True
   )
   > &RigUnit_TransformConstraint(
    BaseBone=None
    BaseTransform=(Rotation=(X=0,Y=0,Z=0,W=1),Translation=(X=0,Y=0,Z=0),Scale3D=(X=1,Y=1,Z=1))
    BaseTransformSpace=GlobalSpace
    Bone=shoulder_r_zero
    Targets=((Transform=(Rotation=(X=0,Y=0,Z=0,W=1),Translation=(X=0,Y=0,Z=0),Scale3D=(X=1,Y=1,Z=1)),Weight=1,bMaintainOffset=True,Filter=(TranslationFilter=(bX=True,bY=True,bZ=True),RotationFilter=(bX=True,bY=True,bZ=True),ScaleFilter=(bX=True,bY=True,bZ=True))))
    bUseInitialTransforms=True
    Targets.0.Transform=&RigUnit_GetBoneTransform(Bone=clavicle_aim_r).Transform
   )
   n32 = &RigUnit_GetRelativeBoneTransform(Bone=shoulderOffset_tgt_r, Space=shoulderOffset_r_zero)
   > &RigUnit_SetBoneTranslation(
    Bone=shoulderOffset_r
    Space=LocalSpace
    Translation=(X=0,Y=0,Z=0)
    Weight=1
    bPropagateToChildren=True
    Translation.X=&RigUnit_GetBoneTransform(Bone=shoulderOffset_r, Space=LocalSpace).Transform.Translation.X
    Translation.Y=@n32.Transform.Translation.Y
    Translation.Z=@n32.Transform.Translation.Z
   )
   > &RigUnit_SetBoneRotation(
    Bone=shoulder_r
    Space=GlobalSpace
    Weight=1
    bPropagateToChildren=True
    Rotation=&RigUnit_GetBoneTransform(Bone=upperarm_r).Transform.Rotation
   )
   > &RigUnit_AimBone(
    Bone=shoulder_aim_r
    DebugSettings=(bEnabled=False,Scale=10,WorldOffset=(Rotation=(X=0,Y=0,Z=0,W=1),Translation=(X=0,Y=0,Z=0),Scale3D=(X=1,Y=1,Z=1)))
    Primary=(Weight=1,Axis=(X=-1,Y=0,Z=0),Target=(X=0,Y=0,Z=0),Kind=Location,Space="lowerarm_r")
    Secondary=(Weight=0,Axis=(X=1,Y=0,Z=0),Target=(X=0,Y=0,Z=0),Kind=Location,Space="None")
    Weight=1
    bPropagateToChildren=True
   )
   > &RigUnit_SetBoneRotation(
    Bone=shoulder_twist_01_r
    Space=LocalSpace
    Weight=1
    bPropagateToChildren=False
    Rotation=&RigUnit_GetBoneTransform(Bone=upperarm_twist_01_r, Space=LocalSpace).Transform.Rotation
   )
  ]}
 SequenceAggregate_1_ContainedGraph {nodes [
   n1 = RigVMFunctionEntryNode()
   SequenceAggregate = &RigVMFunction_Sequence(ExecuteContext=@n1.ExecuteContext)
   SequenceAggregate_1 = &RigVMFunction_Sequence(ExecuteContext=@SequenceAggregate.B)
   SequenceAggregate_2 = &RigVMFunction_Sequence(ExecuteContext=@SequenceAggregate_1.B)
   RigVMFunctionReturnNode(A=@SequenceAggregate.A, B=@SequenceAggregate_1.A, C=@SequenceAggregate_2.A, D=@SequenceAggregate_2.B)
  ]}
]
hierarchy [
 root {loc=<0 0 0>, rot=<0 0 0 1>}
 pelvis {parent=root, loc=<0 -0.7071 93.1191>, rot=<0.0127 0.707 -0.0127 -0.707>}
 spine_01 {parent=pelvis, loc=<6.7494 -0.6388 0>, rot=<0 0 0.034 -0.9994>}
 spine_02 {parent=spine_01, loc=<7.2518 -0.035 0>, rot=<0 0 0.0238 -0.9997>}
 spine_03 {parent=spine_02, loc=<6.0104 0 0>, rot=<0 0 0.0149 -0.9999>}
 spine_04 {parent=spine_03, loc=<8.1024 0 0>, rot=<0 0 -0.0132 -0.9999>}
 spine_05 {parent=spine_04, loc=<10.0146 0 0>, rot=<0 0 -0.0015 -1>}
 clavicle_l {parent=spine_05, loc=<4.3575 -0.1466 -2.3223>, rot=<0.0465 -0.7289 -0.009 -0.6829>}
 upperarm_l {parent=clavicle_l, loc=<15.5111 0 0>, rot=<0.0226 -0.3234 -0.0088 -0.9459>}
 lowerarm_l {parent=upperarm_l, loc=<26.6514 0 0>, rot=<0.0391 -0.0099 -0.2453 -0.9686>}
 hand_l {parent=lowerarm_l, loc=<22.3491 0 0>, rot=<-0.7081 -0.0507 0.0975 -0.6975>}
 index_metacarpal_l {parent=hand_l, loc=<1.6535 0.0183 -2.96>, rot=<0.0158 -0.0124 -0.0173 -0.9996>}
 index_01_l {parent=index_metacarpal_l, loc=<8.3008 0.0357 -0.0524>, rot=<-0.0044 0.0485 0.0924 -0.9945>}
 index_02_l {parent=index_01_l, loc=<3.3587 -0.3003 0.0452>, rot=<-0.002 0.0027 0.3225 -0.9466>}
 index_03_l {parent=index_02_l, loc=<2.5578 -0.123 -0.0252>, rot=<-0.002 0.0067 0.2558 -0.9667>}
 middle_metacarpal_l {parent=hand_l, loc=<2.1856 -0.1725 -0.5984>, rot=<0.006 0.0168 -0.0329 -0.9993>}
 middle_01_l {parent=middle_metacarpal_l, loc=<7.3275 -0.2447 0.0129>, rot=<-0.0208 0.0042 0.0858 -0.9961>}
 middle_02_l {parent=middle_01_l, loc=<4.0123 0.1401 0.1175>, rot=<-0.0218 0.0258 0.3512 -0.9357>}
 middle_03_l {parent=middle_02_l, loc=<2.6077 0.0534 0.1373>, rot=<0.0082 0.0188 0.1708 -0.9851>}
 pinky_metacarpal_l {parent=hand_l, loc=<1.7616 -0.9753 3.3814>, rot=<-0.0397 0.0814 -0.069 -0.9935>}
 pinky_01_l {parent=pinky_metacarpal_l, loc=<7.1709 -0.4831 -0.1094>, rot=<-0.0654 -0.0035 0.2242 -0.9723>}
 pinky_02_l {parent=pinky_01_l, loc=<2.8274 0.0893 -0.1043>, rot=<0.0037 0.0011 0.2854 -0.9584>}
 pinky_03_l {parent=pinky_02_l, loc=<2.2046 0.0308 -0.0078>, rot=<-0.0002 0.0098 0.1445 -0.9895>}
 ring_metacarpal_l {parent=hand_l, loc=<2.2074 -0.3625 1.3638>, rot=<-0.0284 0.0722 -0.0458 -0.9959>}
 ring_01_l {parent=ring_metacarpal_l, loc=<7.0295 -0.6197 -0.1446>, rot=<-0.0489 -0.0252 0.1892 -0.9804>}
 ring_02_l {parent=ring_01_l, loc=<3.6174 0.4646 -0.1027>, rot=<0 0 0.2806 -0.9598>}
 ring_03_l {parent=ring_02_l, loc=<2.4152 0.1141 0.0199>, rot=<0 0 0.1189 -0.9929>}
 thumb_01_l {parent=hand_l, loc=<1.3118 -0.3191 -2.459>, rot=<0.6546 -0.4042 0.0664 -0.6354>}
 thumb_02_l {parent=thumb_01_l, loc=<5.9336 -0.0448 0.0401>, rot=<-0.0037 0.0213 0.298 -0.9543>}
 thumb_03_l {parent=thumb_02_l, loc=<2.2223 0.3682 -0.0907>, rot=<0.0118 -0.013 0.0066 -0.9998>}
 lowerarm_twist_01_l {parent=lowerarm_l, loc=<17.8549 0 0>, rot=<0 0 0 -1>}
 lowerarm_twist_02_l {parent=lowerarm_l, loc=<9.6199 0 0>, rot=<0 0 0 -1>}
 armBezier_2_l_zero {parent=lowerarm_l, loc=<-6.1129 0.6 -0.1923>, rot=<-0.0099 -0.0391 0.9686 0.2453>}
 armBezier_2_l {parent=armBezier_2_l_zero, loc=<0 0 0>, rot=<0 0 0 1>}
 armBezier_3_l_zero {parent=lowerarm_l, loc=<2.416 -2.9628 0.0001>, rot=<0 0 1 0>}
 armBezier_3_l {parent=armBezier_3_l_zero, loc=<0 0 0>, rot=<0 0 0 -1>}
 upperarm_twist_01_l {parent=upperarm_l, loc=<0.0041 0 0>, rot=<0 0 0 -1>}
 noodle_0_l {parent=upperarm_twist_01_l, loc=<-0.0042 0 0.0001>, rot=<-0.002 -0.0395 0.9982 0.0457>}
 noodle_1_l_zero {parent=upperarm_twist_01_l, loc=<-0.0042 0 0.0001>, rot=<-0.002 -0.0395 0.9982 0.0457>}
 noodle_1_l {parent=noodle_1_l_zero, loc=<-6.9718 0.762 -0.1271>, rot=<0 -0.004 -0.0239 -0.9997>}
 noodle_2_l {parent=noodle_1_l, loc=<-6.7098 0 0>, rot=<0 0.0062 -0.0444 -0.999>}
 noodle_3_l {parent=noodle_2_l, loc=<-6.2183 0 0>, rot=<0 0.0108 -0.0678 -0.9976>}
 noodle_4_l {parent=noodle_3_l, loc=<-6.1517 0 0>, rot=<0 0.0133 -0.0783 -0.9968>}
 noodle_5_l {parent=noodle_4_l, loc=<-6.5967 0 0>, rot=<0 0.0129 -0.0718 -0.9973>}
 noodle_6_l {parent=noodle_5_l, loc=<-7.5957 0 0>, rot=<0 0.0104 -0.0557 -0.9984>}
 noodle_7_l {parent=noodle_6_l, loc=<-9.1342 0 0>, rot=<0 0 0 1>}
 upperarm_twist_02_l {parent=upperarm_l, loc=<13.7941 0 0>, rot=<0 0 0 -1>}
 shoulder_l_zero {parent=upperarm_l, loc=<6.8741 -1.3839 -0.1838>, rot=<0 0 0 -1>}
 shoulderOffset_l_zero {parent=shoulder_l_zero, loc=<0 0 0>, rot=<-0.0134 0.3555 -0.014 -0.9345>}
 shoulderOffset_l {parent=shoulderOffset_l_zero, loc=<0 0 0>, rot=<0 0 0 1>}
 shoulder_l {parent=shoulderOffset_l, loc=<0 0 0>, rot=<0.0134 -0.3555 0.014 -0.9345>}
 shoulder_aim_l_zero {parent=shoulder_l, loc=<0 0 0>, rot=<0.0009 0.0047 -0.0349 -0.9994>}
 shoulder_aim_l {parent=shoulder_aim_l_zero, loc=<0 0 0>, rot=<0 0 0 -1>}
 shoulder_twist_01_l {parent=shoulder_aim_l, loc=<0 0 0>, rot=<0 0 0 -1>}
 shoulder_twist_02_l {parent=shoulder_aim_l, loc=<0 0 0>, rot=<0 0 0 -1>}
 shoulderOffset_tgt_l {parent=upperarm_l, loc=<6.8741 -1.3839 -0.1838>, rot=<-0.0134 0.3555 -0.014 -0.9345>}
 upperarm_follow_l_zero {parent=clavicle_l, loc=<15.5111 0 0>, rot=<0.0008 0.0336 -0.0252 -0.9991>}
 upperarm_follow_l {parent=upperarm_follow_l_zero, loc=<0 0 0>, rot=<0 0 0 1>}
 clavicle_tgt_l {parent=upperarm_follow_l, loc=<-0.67 0.0338 -9.9775>, rot=<-0.0008 -0.0336 0.0252 -0.9991>}
 clavicle_aim_l {parent=upperarm_follow_l, loc=<0 0 0>, rot=<-0.0008 -0.0336 0.0252 -0.9991>}
 armDirection_l {parent=clavicle_aim_l, loc=<0 0 0.0001>, rot=<0 -0.3308 -0.1197 -0.9361>}
 clavicle_twist_l {parent=clavicle_aim_l, loc=<0 0 0.0001>, rot=<0 0 0 -1>}
 shldrFront_l_zero {parent=clavicle_aim_l, loc=<1.708 4.5266 -3.1377>, rot=<0.0008 0.0336 -0.0252 -0.9991>}
 shldrFront_l {parent=shldrFront_l_zero, loc=<0 0 0>, rot=<0 0 0 1>}
 shldrBack_l_zero {parent=clavicle_aim_l, loc=<2.031 -5.5552 -4.2732>, rot=<0.0008 0.0336 -0.0252 -0.9991>}
 shldrBack_l {parent=shldrBack_l_zero, loc=<0 0 0>, rot=<0 0 0 1>}
 clavicle_r {parent=spine_05, loc=<4.3575 -0.1466 2.3223>, rot=<0.7289 0.0465 -0.6829 0.009>}
 upperarm_r {parent=clavicle_r, loc=<-15.5111 0 0>, rot=<0.0226 -0.3234 -0.0088 -0.9459>}
 lowerarm_r {parent=upperarm_r, loc=<-26.6515 0 0>, rot=<0.0391 -0.0099 -0.2453 -0.9686>}
 hand_r {parent=lowerarm_r, loc=<-22.3492 0 0>, rot=<-0.7081 -0.0507 0.0975 -0.6975>}
 index_metacarpal_r {parent=hand_r, loc=<-1.6534 -0.0181 2.9601>, rot=<0.0158 -0.0124 -0.0173 -0.9996>}
 index_01_r {parent=index_metacarpal_r, loc=<-8.3008 -0.0357 0.0524>, rot=<-0.0044 0.0485 0.0924 -0.9945>}
 index_02_r {parent=index_01_r, loc=<-3.3589 0.3003 -0.0453>, rot=<-0.002 0.0027 0.3225 -0.9466>}
 index_03_r {parent=index_02_r, loc=<-2.5578 0.1231 0.0251>, rot=<-0.002 0.0067 0.2558 -0.9667>}
 middle_metacarpal_r {parent=hand_r, loc=<-2.1856 0.1726 0.5984>, rot=<0.006 0.0168 -0.0329 -0.9993>}
 middle_01_r {parent=middle_metacarpal_r, loc=<-7.3273 0.2447 -0.0129>, rot=<-0.0208 0.0042 0.0858 -0.9961>}
 middle_02_r {parent=middle_01_r, loc=<-4.0142 -0.1411 -0.07>, rot=<-0.0217 0.0259 0.3512 -0.9357>}
 middle_03_r {parent=middle_02_r, loc=<-2.6074 -0.0534 -0.1312>, rot=<0.0083 0.0188 0.1708 -0.9851>}
 pinky_metacarpal_r {parent=hand_r, loc=<-1.7617 0.9753 -3.3814>, rot=<-0.0397 0.0814 -0.069 -0.9935>}
 pinky_01_r {parent=pinky_metacarpal_r, loc=<-7.1707 0.4832 0.1094>, rot=<-0.0654 -0.0035 0.2242 -0.9723>}
 pinky_02_r {parent=pinky_01_r, loc=<-2.8274 -0.0892 0.1043>, rot=<0.0037 0.0011 0.2854 -0.9584>}
 pinky_03_r {parent=pinky_02_r, loc=<-2.2045 -0.031 0.0078>, rot=<-0.0002 0.0098 0.1445 -0.9895>}
 ring_metacarpal_r {parent=hand_r, loc=<-2.2073 0.3627 -1.3638>, rot=<-0.0284 0.0722 -0.0458 -0.9959>}
 ring_01_r {parent=ring_metacarpal_r, loc=<-7.0297 0.6197 0.1446>, rot=<-0.0489 -0.0252 0.1892 -0.9804>}
 ring_02_r {parent=ring_01_r, loc=<-3.6172 -0.4648 0.1027>, rot=<0 0 0.2806 -0.9598>}
 ring_03_r {parent=ring_02_r, loc=<-2.4153 -0.1141 -0.0199>, rot=<0 0 0.1189 -0.9929>}
 thumb_01_r {parent=hand_r, loc=<-1.3119 0.3192 2.459>, rot=<0.6546 -0.4042 0.0664 -0.6354>}
 thumb_02_r {parent=thumb_01_r, loc=<-5.9336 0.0448 -0.0401>, rot=<-0.0037 0.0213 0.298 -0.9543>}
 thumb_03_r {parent=thumb_02_r, loc=<-2.2223 -0.3682 0.0907>, rot=<0.0118 -0.013 0.0066 -0.9998>}
 lowerarm_twist_01_r {parent=lowerarm_r, loc=<-17.855 0 0>, rot=<0 0 0 -1>}
 lowerarm_twist_02_r {parent=lowerarm_r, loc=<-9.6199 0 0>, rot=<0 0 0 -1>}
 armBezier_3_r_zero {parent=lowerarm_r, loc=<-2.416 2.9628 0>, rot=<0 0 0 -1>}
 armBezier_3_r {parent=armBezier_3_r_zero, loc=<0 0 0>, rot=<0 0 0 1>}
 armBezier_2_r_zero {parent=lowerarm_r, loc=<6.1129 -0.6 0.1924>, rot=<-0.0391 0.0099 0.2453 -0.9686>}
 armBezier_2_r {parent=armBezier_2_r_zero, loc=<0 0 0>, rot=<0 0 0 1>}
 upperarm_twist_01_r {parent=upperarm_r, loc=<-0.0041 0 0>, rot=<0 0 0 1>}
 noodle_1_r {parent=upperarm_twist_01_r, loc=<-6.8701 1.384 0.1839>, rot=<0 0.0061 0.0222 -0.9997>}
 noodle_2_r {parent=noodle_1_r, loc=<-6.7098 0 0>, rot=<0 -0.0027 -0.0448 -0.999>}
 noodle_3_r {parent=noodle_2_r, loc=<-6.2183 0 0>, rot=<0 -0.0054 -0.0684 -0.9976>}
 noodle_4_r {parent=noodle_3_r, loc=<-6.1517 0 0>, rot=<0 -0.0072 -0.0791 -0.9968>}
 noodle_5_r {parent=noodle_4_r, loc=<-6.5967 0 0>, rot=<0 -0.0072 -0.0726 -0.9973>}
 noodle_6_r {parent=noodle_5_r, loc=<-7.5958 0 0>, rot=<0 -0.006 -0.0563 -0.9984>}
 noodle_7_r {parent=noodle_6_r, loc=<-9.1343 0 0>, rot=<0 0 0 1>}
 noodle_0_r {parent=upperarm_twist_01_r, loc=<0.0041 0 0>, rot=<-0.0395 0.002 0.0457 -0.9982>}
 upperarm_twist_02_r {parent=upperarm_r, loc=<-13.7942 0 0>, rot=<0 0 0 1>}
 shoulder_r_zero {parent=upperarm_r, loc=<-6.8742 1.384 0.1839>, rot=<0 0 0 1>}
 shoulderOffset_r_zero {parent=shoulder_r_zero, loc=<0 0 0>, rot=<0.9345 0.014 0.3555 -0.0134>}
 shoulderOffset_r {parent=shoulderOffset_r_zero, loc=<0 0 0>, rot=<0 0 0 1>}
 shoulder_r {parent=shoulderOffset_r, loc=<0 0 0>, rot=<0.9345 0.014 0.3555 0.0134>}
 shoulder_aim_r_zero {parent=shoulder_r, loc=<0 0 0>, rot=<0.0009 0.0047 -0.0349 -0.9994>}
 shoulder_aim_r {parent=shoulder_aim_r_zero, loc=<0 0 0>, rot=<0 0 0 1>}
 shoulder_twist_02_r {parent=shoulder_aim_r, loc=<0 0 0>, rot=<0 0 0 1>}
 shoulder_twist_01_r {parent=shoulder_aim_r, loc=<0 0 0>, rot=<0 0 0 1>}
 shoulderOffset_tgt_r {parent=upperarm_r, loc=<-6.8742 1.384 0.1839>, rot=<0.9345 0.014 0.3555 -0.0134>}
 upperarm_follow_r_zero {parent=clavicle_r, loc=<-15.5111 0 0.0001>, rot=<-0.0336 0.0008 0.9991 -0.0252>}
 upperarm_follow_r {parent=upperarm_follow_r_zero, loc=<0 0 -0.0001>, rot=<0 0 0 1>}
 clavicle_aim_r {parent=upperarm_follow_r, loc=<0 0 0>, rot=<0.0008 0.0336 0.0252 -0.9991>}
 clavicle_twist_r {parent=clavicle_aim_r, loc=<0 0 0>, rot=<0 0 0 -1>}
 armDirection_r {parent=clavicle_aim_r, loc=<0 0 0>, rot=<0 0.3308 -0.1197 -0.9361>}
 shldrFront_r_zero {parent=clavicle_aim_r, loc=<1.708 4.5266 3.1378>, rot=<-0.0008 -0.0336 -0.0252 -0.9991>}
 shldrFront_r {parent=shldrFront_r_zero, loc=<0 0 0>, rot=<0 0 0 1>}
 shldrBack_r_zero {parent=clavicle_aim_r, loc=<2.031 -5.5552 4.2733>, rot=<-0.0008 -0.0336 -0.0252 -0.9991>}
 shldrBack_r {parent=shldrBack_r_zero, loc=<0 0 0>, rot=<0 0 0 1>}
 clavicle_tgt_r {parent=upperarm_follow_r, loc=<-0.67 0.0338 9.9775>, rot=<0.0008 0.0336 0.0252 -0.9991>}
 neck_01 {parent=spine_05, loc=<11.3138 0.1529 0>, rot=<0 0 -0.2003 -0.9797>}
 neck_02 {parent=neck_01, loc=<4.5627 0 0>, rot=<0 0 0.0504 -0.9987>}
 head {parent=neck_02, loc=<5.3914 0 0>, rot=<0 0 0.115 -0.9934>}
 headTop_zero {parent=neck_01, loc=<-5.2331 2.1497 0>, rot=<0 0 0 1>}
 headTop {parent=headTop_zero, loc=<0 0 0>, rot=<0 0 0 1>}
 necklace_l_zero {parent=headTop, loc=<6.4814 7.271 -8.3541>, rot=<0.6781 -0.2257 0.6812 0.1588>}
 necklace_l {parent=necklace_l_zero, loc=<0 0 0>, rot=<0 0 0 1>}
 necklace_c_zero {parent=headTop, loc=<6.736 7.5271 0>, rot=<0 0 0.932 0.3625>}
 necklace_c {parent=necklace_c_zero, loc=<0 0 0>, rot=<0 0 0 -1>}
 necklace_r_zero {parent=headTop, loc=<6.736 7.5271 8.2196>, rot=<-0.2009 -0.4361 0.4751 -0.7374>}
 necklace_r {parent=necklace_r_zero, loc=<0 0 0>, rot=<0 0 0 -1>}
 headTop_bun {parent=headTop, loc=<0 0 0>, rot=<0 0 0 1>}
 headTop_bun_r {parent=headTop_bun, loc=<0 0 0>, rot=<0 0 0 1>}
 relish_04_zero {parent=headTop_bun_r, loc=<14.6777 -7.0752 11.5416>, rot=<-0.1321 -0.8108 0.5628 0.0917>}
 relish_04 {parent=relish_04_zero, loc=<0 0 0>, rot=<0 0 0 1>}
 relish_05_zero {parent=headTop_bun_r, loc=<14.1497 -4.1813 12.1932>, rot=<-0.1215 -0.7454 0.647 0.1054>}
 relish_05 {parent=relish_05_zero, loc=<0 0 0>, rot=<0 0 0 1>}
 relish_06_zero {parent=headTop_bun_r, loc=<14.2954 -0.929 12.5064>, rot=<-0.1072 -0.658 0.7357 0.1199>}
 relish_06 {parent=relish_06_zero, loc=<0 0 0>, rot=<0 0 0 1>}
 relish_07_zero {parent=headTop_bun_r, loc=<13.5163 2.0599 10.7584>, rot=<-0.0919 -0.5638 0.8101 0.132>}
 relish_07 {parent=relish_07_zero, loc=<0 0 0>, rot=<0 0 0 1>}
 relish_08_zero {parent=headTop_bun_r, loc=<11.4336 5.9313 9.2845>, rot=<-0.5598 0.1684 -0.6287 -0.5128>}
 relish_08 {parent=relish_08_zero, loc=<0 0 0>, rot=<0 0 0 1>}
 relish_09_zero {parent=headTop_bun_r, loc=<7.0675 9.2396 8.9956>, rot=<0.7043 0.0632 0.6478 0.2835>}
 relish_09 {parent=relish_09_zero, loc=<0 0 0>, rot=<0 0 0 1>}
 headTop_bun_l {parent=headTop_bun, loc=<0 0 0>, rot=<0 0 0 1>}
 headTop_bun_c {parent=headTop_bun, loc=<0 0 0>, rot=<0 0 0 1>}
 relish_03_zero {parent=headTop_bun_c, loc=<13.5137 -10.4267 10.0123>, rot=<-0.1446 -0.8875 0.4317 0.0704>}
 relish_03 {parent=relish_03_zero, loc=<0 0 0>, rot=<0 0 0 -1>}
 relish_02_zero {parent=headTop_bun_c, loc=<13.8411 -12.1464 7.2122>, rot=<-0.1523 -0.9345 0.3176 0.0518>}
 relish_02 {parent=relish_02_zero, loc=<0 0 0>, rot=<0 0 0 -1>}
 relish_01_zero {parent=headTop_bun_c, loc=<13.7953 -13.6907 2.3731>, rot=<-0.1587 -0.9737 0.1612 0.0263>}
 relish_01 {parent=relish_01_zero, loc=<0 0 0>, rot=<0 0 0 1>}
 neck01_bun {parent=neck_01, loc=<0 0 0>, rot=<0 0 0 1>}
 neck01_bun_r {parent=neck01_bun, loc=<0 0 0>, rot=<0 0 0 1>}
 neck01_bun_l {parent=neck01_bun, loc=<0 0 0>, rot=<0 0 0 1>}
 neck01_bun_c {parent=neck01_bun, loc=<0 0 0>, rot=<0 0 0 1>}
 spine_05_bun {parent=spine_05, loc=<0 0 0>, rot=<0 0 0 1>}
 spine_05_bun_r {parent=spine_05_bun, loc=<0 0 0>, rot=<0 0 0 1>}
 spine_05_bun_l {parent=spine_05_bun, loc=<0 0 0>, rot=<0 0 0 1>}
 spine_05_bun_c {parent=spine_05_bun, loc=<0 0 0>, rot=<0 0 0 1>}
 bunTop_c_zero {parent=spine_05_bun_c, loc=<5.4394 -4.7463 0>, rot=<0.0352 0.0002 -0.0049 -0.9994>}
 bunTop_c {parent=bunTop_c_zero, loc=<0 0 0>, rot=<0 0 0 1>}
 spine_04_bun {parent=spine_04, loc=<0 0 0>, rot=<0 0 0 1>}
 spine_04_bun_r {parent=spine_04_bun, loc=<0 0 0>, rot=<0 0 0 1>}
 spine_04_bun_l {parent=spine_04_bun, loc=<0 0 0>, rot=<0 0 0 1>}
 spine_03_bun {parent=spine_03, loc=<0 0 0>, rot=<0 0 0 -1>}
 spine_03_bun_r {parent=spine_03_bun, loc=<0 0 0>, rot=<0 0 0 -1>}
 spine_03_bun_l {parent=spine_03_bun, loc=<0 0 0>, rot=<0 0 0 -1>}
 spine_02_bun {parent=spine_02, loc=<0 0 0>, rot=<0 0 0 1>}
 spine_02_bun_r {parent=spine_02_bun, loc=<0 0 0>, rot=<0 0 0 1>}
 spine_02_bun_l {parent=spine_02_bun, loc=<0 0 0>, rot=<0 0 0 1>}
 spine_01_bun {parent=spine_01, loc=<0 0 0>, rot=<0 0 0 -1>}
 spine_01_bun_r {parent=spine_01_bun, loc=<0 0 0>, rot=<0 0 0 -1>}
 spine_01_bun_l {parent=spine_01_bun, loc=<0 0 0>, rot=<0 0 0 -1>}
 thigh_l {parent=pelvis, loc=<-2.5693 -0.0109 -7.7501>, rot=<0.0012 0.0572 0.0155 -0.9982>}
 calf_l {parent=thigh_l, loc=<-39.0484 0 0>, rot=<0.0006 -0.0161 -0.0391 -0.9991>}
 calf_twist_01_l {parent=calf_l, loc=<-35.8131 -0.0967 0.0397>, rot=<0.0237 0.0095 -0.0035 -0.9997>}
 calf_twist_02_l {parent=calf_l, loc=<-18.8064 -0.0967 0.0397>, rot=<0.0237 0.0095 -0.0035 -0.9997>}
 foot_l {parent=calf_l, loc=<-42.561 0.1077 0.4292>, rot=<0.0418 -0.0455 0.0343 -0.9975>}
 ball_l {parent=foot_l, loc=<-6.1599 13.8264 0.6806>, rot=<0.0001 -0.0001 -0.7063 -0.7079>}
 lacesOut_l {parent=foot_l, loc=<1.3213 7.5845 0.3445>, rot=<-0.0284 0.7082 0.0357 -0.7045>}
 lacesInner_l {parent=foot_l, loc=<1.0001 7.3343 0.5738>, rot=<0.7045 -0.0357 0.7082 -0.0284>}
 pantsTatters_1_l_zero {parent=calf_l, loc=<-2.8289 1.9144 -0.4317>, rot=<-0.0057 0.0457 0.9935 0.1041>}
 pantsTatters_1_l {parent=pantsTatters_1_l_zero, loc=<0 0 0>, rot=<0 0 0 1>}
 thigh_twist_01_l {parent=thigh_l, loc=<-21.6303 0 0>, rot=<0.0013 0 0 -1>}
 thigh_r {parent=pelvis, loc=<-2.5691 -0.0109 7.7501>, rot=<0.0572 -0.0012 0.9982 0.0155>}
 calf_r {parent=thigh_r, loc=<39.0487 0 0>, rot=<0.0006 -0.0161 -0.0391 -0.9991>}
 calf_twist_01_r {parent=calf_r, loc=<35.8132 0.0967 -0.0397>, rot=<0.0237 0.0095 -0.0035 -0.9997>}
 calf_twist_02_r {parent=calf_r, loc=<18.8064 0.0967 -0.0397>, rot=<0.0237 0.0095 -0.0035 -0.9997>}
 pantsTatters_2_r_zero {parent=calf_twist_02_r, loc=<-7.71 -1.5598 -0.1028>, rot=<-0.0659 -0.0487 0.0459 -0.9956>}
 pantsTatters_2_r {parent=pantsTatters_2_r_zero, loc=<0 0 0>, rot=<0 0 0 1>}
 pantsTatters_3_r_zero {parent=calf_twist_02_r, loc=<-4.6962 -1.0568 2.8941>, rot=<0.2007 0.0605 0.0921 -0.9734>}
 pantsTatters_3_r {parent=pantsTatters_3_r_zero, loc=<0 0 0>, rot=<0 0 0 -1>}
 foot_r {parent=calf_r, loc=<42.5611 -0.1077 -0.4292>, rot=<0.0418 -0.0455 0.0343 -0.9975>}
 ball_r {parent=foot_r, loc=<6.1598 -13.8264 -0.6806>, rot=<0.0001 -0.0001 -0.7063 -0.7079>}
 lacesOut_r {parent=foot_r, loc=<-1.3215 -7.5845 -0.3445>, rot=<0.7082 0.0284 -0.7045 -0.0357>}
 lacesInner_r {parent=foot_r, loc=<-1.0002 -7.3343 -0.5738>, rot=<-0.0357 -0.7045 -0.0284 -0.7082>}
 pantsTatters_1_r_zero {parent=calf_r, loc=<4.2121 -2.7507 0.1874>, rot=<0.0156 -0.0061 0.0286 -0.9994>}
 pantsTatters_1_r {parent=pantsTatters_1_r_zero, loc=<0 0 0>, rot=<0 0 0 1>}
 thigh_twist_01_r {parent=thigh_r, loc=<21.6303 0 0>, rot=<0.0013 0 0 -1>}
 bun_hip_1_l {parent=pelvis, loc=<13.2204 -2.6625 -1.127>, rot=<-0.0464 0.4286 0.8028 0.4119>}
 bun_hip_2_l {parent=bun_hip_1_l, loc=<15.5954 2.3536 -12.1711>, rot=<-0.0464 0.4286 0.8028 -0.4119>}
 bun_hip_3_l {parent=bun_hip_2_l, loc=<0 13.9558 3.0007>, rot=<0 0 0 -1>}
 pelvis_bun_l {parent=bun_hip_3_l, loc=<2.0569 -11.4557 10.9118>, rot=<0 0 0 -1>}
 bun_hip_1_r {parent=pelvis, loc=<13.2204 -2.6625 1.127>, rot=<0.0464 -0.4286 0.8028 0.4119>}
 bun_hip_2_r {parent=bun_hip_1_r, loc=<15.5954 2.3536 12.1711>, rot=<0.0464 -0.4286 0.8028 -0.4119>}
 bun_hip_3_r {parent=bun_hip_2_r, loc=<0 13.9558 -3.0007>, rot=<0 0 0 -1>}
 pelvis_bun_r {parent=bun_hip_3_r, loc=<2.0569 -11.4557 -10.9118>, rot=<0 0 0 -1>}
 beltFlap_zero {parent=pelvis, loc=<-0.2842 9.992 -2.817>, rot=<-0.405 -0.0038 0.9142 0.0115>}
 beltFlap {parent=beltFlap_zero, loc=<0 0 0>, rot=<0 0 0 1>}
 pelvis_bun {parent=pelvis, loc=<0 0 0>, rot=<0 0 0 -1>}
 pelvis_bun_c {parent=pelvis, loc=<0 0 0>, rot=<0 0 0 -1>}
 bunJiggle_c_zero {parent=pelvis, loc=<5.1679 -5.1793 0>, rot=<-0.0127 -0.707 0.0127 -0.707>}
 bunJiggle_c {parent=bunJiggle_c_zero, loc=<0 0 0>, rot=<0 -0.7071 0 -0.7071>}
 bunBottom_c_zero {parent=bunJiggle_c, loc=<5.3501 0 0>, rot=<0.0352 -0.9988 -0.0352 0.0012>}
 bunBottom_c {parent=bunBottom_c_zero, loc=<0 0 0>, rot=<0 0 0 1>}
]
variables [
 LOD2 {type=double, default=0}
 LOD2_0 {type=double, default=0}
 target_alpha {type=double, default=0}
]
