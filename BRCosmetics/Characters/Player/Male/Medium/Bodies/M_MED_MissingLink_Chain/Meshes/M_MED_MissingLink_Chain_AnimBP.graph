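# Post-process style AnimBlueprint for the Missing Link Chain male cosmetic body, layering secondary motion onto the pose copied from its attached parent mesh.
# Parent constraints and the shared Missing Link Control Rig feed alpha-driven hip corrections, then weighted head-top and necklace constraints.
# Spring-based AnimDynamics jiggles necklace, tatters, laces, belt flap and relish bones, with a bone driver copying spine swing to headTop.
# TargetingAlpha drives the rig and final interpolated right-shoulder offsets.
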AnimBlueprint /BRCosmetics/Characters/Player/Male/Medium/Bodies/M_MED_MissingLink_Chain/Meshes/M_MED_MissingLink_Chain_AnimBP
target_skeleton=/BRCosmetics/Characters/Player/Male/Medium/Bodies/M_MED_Missing_Link/Meshes/M_MED_Missing_Link_Skeleton
graphs [
 AnimGraph {nodes [
   n1 = "Constraint - upperarm_follow_r_zero (P)"(ComponentPose=Local To Component(LocalPose=Copy Pose From Mesh() {bUseAttachedParent=true}), Alpha=1, ConstraintWeights_0=1) {
    alphaBoolBlend {blendOption=Linear}
    boneToModify {boneName=upperarm_follow_r_zero}
    constraintSetup [{targetBone {boneName=spine_05}}]
    constraintWeights [1]
   }
   n2 = "Constraint - upperarm_follow_l_zero (P)"(ComponentPose=@n1, Alpha=1, ConstraintWeights_0=1) {
    alphaBoolBlend {blendOption=Linear}
    boneToModify {boneName=upperarm_follow_l_zero}
    constraintSetup [{targetBone {boneName=spine_05}}]
    constraintWeights [1]
   }
   n3 = Control Rig!(
    Source=Component To Local(ComponentPose=@n2)
    Alpha
    target_alpha:double=Get TargetingAlpha() -> (TargetingAlpha:float)
    LOD2:double=Get LOD2Alpha() -> (LOD2Alpha:float)
   ) -> (Pose) {
    alphaBoolBlend {blendOption=Linear}
    controlRigAssetReference {blueprintRigClass=/BRCosmetics/Characters/Player/Male/Medium/Bodies/M_MED_Missing_Link/Meshes/M_MED_Missing_Link_CtrlRig.M_MED_Missing_Link_CtrlRig_C}
   }
   n4 = "Transform (Modify) Bone - Bone: bun_hip_1_l"(
    ComponentPose=Local To Component(LocalPose=@n3)
    Alpha=Get ThighLeftRotationAlpha() -> (ThighLeftRotationAlpha:float)
    Translation="0,0,0"
    Rotation="20.000,-30,0"
    Scale="1,1,1"
   ) {
    rotationMode=BMM_Additive
    translationMode=BMM_Additive
    translationSpace=BCS_BoneSpace
    alphaBoolBlend {blendOption=Linear}
    alphaScaleBiasClamp {bInterpResult=true}
    boneToModify {boneName=bun_hip_1_l}
    rotation {pitch=20, yaw=-30}
   }
   n5 = "Transform (Modify) Bone - Bone: bun_hip_1_r"(
    ComponentPose=@n4
    Alpha=Get ThighRightRotationAlpha() -> (ThighRightRotationAlpha:float)
    Translation="0,0,3"
    Rotation="-20,30,0.000"
    Scale="1,1,1"
   ) {
    rotationMode=BMM_Additive
    translationMode=BMM_Additive
    translationSpace=BCS_ParentBoneSpace
    alphaBoolBlend {blendOption=Linear}
    alphaScaleBiasClamp {bInterpResult=true, scale=0.7}
    boneToModify {boneName=bun_hip_1_r}
    rotation {pitch=-20, yaw=30}
    translation {z=3}
   }
   n6 = "Constraint - headTop (PP)"!(ComponentPose=@n5, Alpha=1, ConstraintWeights_0=0.5, ConstraintWeights_1:float=0.5) -> (Pose) {
    alphaBoolBlend {blendOption=Linear}
    boneToModify {boneName=headTop}
    constraintSetup [{targetBone {boneName=spine_05}}, {targetBone {boneName=neck_01}}]
    constraintWeights [1,1]
   }
   n7 = "Constraint - necklace_c_zero (PP)"!(ComponentPose=@n6, Alpha=1, ConstraintWeights_0=0.75, ConstraintWeights_1:float=0.25) -> (Pose) {
    alphaBoolBlend {blendOption=Linear}
    boneToModify {boneName=necklace_c_zero}
    constraintSetup [{targetBone {boneName=headTop}}, {targetBone {boneName=spine_04}}]
    constraintWeights [1,1]
   }
   n8 = Anim Dynamics - Bone: necklace_c(ComponentPose=@n7, Alpha=1, ExternalForce="0,0,0") {
    angularDampingOverride=1
    angularSpringConstant=1500
    bAngularSpring=true
    bGravityOverrideInSimSpace=true
    bOverrideAngularDamping=true
    gravityScale=0.35
    simulationSpace=RootRelative
    alphaBoolBlend {blendOption=Linear}
    boundBone {boneName=necklace_c}
    physicsBodyDefinitions [{
      sphereCollisionRadius=0
      boundBone {boneName=necklace_c}
      boxExtents=<10 20 20>
      constraintSetup {angularLimitsMax=<0 20 0>, angularLimitsMin=<0 -20 -60>, angularTarget=<1 0 0>}
      localJointOffset=<10 0 0>
     }]
   }
   n9 = Anim Dynamics - Bone: pantsTatters_1_l(ComponentPose=@n8, Alpha=1, ExternalForce="0,0,0") {
    angularDampingOverride=1
    angularSpringConstant=1500
    bAngularSpring=true
    bGravityOverrideInSimSpace=true
    bOverrideAngularDamping=true
    gravityScale=0.35
    simulationSpace=RootRelative
    alphaBoolBlend {blendOption=Linear}
    boundBone {boneName=pantsTatters_1_l}
    physicsBodyDefinitions [{
      sphereCollisionRadius=0
      boundBone {boneName=pantsTatters_1_l}
      boxExtents=<10 20 20>
      constraintSetup {angularLimitsMax=<0 0 0>, angularLimitsMin=<0 0 -40>, angularTarget=<0 0 0>}
      localJointOffset=<10 0 0>
     }]
   }
   n10 = Anim Dynamics - Bone: pantsTatters_1_r(ComponentPose=@n9, Alpha=1, ExternalForce="0,0,0") {
    angularDampingOverride=1
    angularSpringConstant=1500
    bAngularSpring=true
    bGravityOverrideInSimSpace=true
    bOverrideAngularDamping=true
    gravityScale=0.35
    simulationSpace=RootRelative
    alphaBoolBlend {blendOption=Linear}
    boundBone {boneName=pantsTatters_1_r}
    physicsBodyDefinitions [{
      sphereCollisionRadius=0
      boundBone {boneName=pantsTatters_1_r}
      boxExtents=<10 20 20>
      constraintSetup {angularLimitsMax=<0 0 0>, angularLimitsMin=<0 0 -40>, angularTarget=<0 0 0>}
      localJointOffset=<10 0 0>
     }]
   }
   n11 = Anim Dynamics - Bone: pantsTatters_2_r(ComponentPose=@n10, Alpha=1, ExternalForce="0,0,0") {
    angularDampingOverride=1
    angularSpringConstant=1500
    bAngularSpring=true
    bGravityOverrideInSimSpace=true
    bOverrideAngularDamping=true
    gravityScale=0.35
    simulationSpace=RootRelative
    alphaBoolBlend {blendOption=Linear}
    boundBone {boneName=pantsTatters_2_r}
    physicsBodyDefinitions [{
      sphereCollisionRadius=0
      boundBone {boneName=pantsTatters_2_r}
      boxExtents=<10 20 20>
      constraintSetup {angularLimitsMax=<0 0 0>, angularLimitsMin=<0 0 -40>, angularTarget=<0 0 0>}
      localJointOffset=<10 0 0>
     }]
   }
   n12 = Anim Dynamics - Bone: pantsTatters_3_r(ComponentPose=@n11, Alpha=1, ExternalForce="0,0,0") {
    angularDampingOverride=1
    angularSpringConstant=1500
    bAngularSpring=true
    bGravityOverrideInSimSpace=true
    bOverrideAngularDamping=true
    gravityScale=0.35
    simulationSpace=RootRelative
    alphaBoolBlend {blendOption=Linear}
    boundBone {boneName=pantsTatters_3_r}
    physicsBodyDefinitions [{
      sphereCollisionRadius=0
      boundBone {boneName=pantsTatters_3_r}
      boxExtents=<10 20 20>
      constraintSetup {angularLimitsMax=<0 0 0>, angularLimitsMin=<0 0 -40>, angularTarget=<0 0 0>}
      localJointOffset=<10 0 0>
     }]
   }
   n13 = Anim Dynamics - Bone: lacesOut_r(ComponentPose=@n12, Alpha=1, ExternalForce="0,0,0") {
    angularDampingOverride=1
    angularSpringConstant=1500
    bAngularSpring=true
    bGravityOverrideInSimSpace=true
    bOverrideAngularDamping=true
    gravityScale=0.35
    simulationSpace=RootRelative
    alphaBoolBlend {blendOption=Linear}
    alphaScaleBiasClamp {bInterpResult=true}
    boundBone {boneName=lacesOut_r}
    physicsBodyDefinitions [{
      sphereCollisionRadius=0
      boundBone {boneName=lacesOut_r}
      boxExtents=<20 10 10>
      constraintSetup {angularLimitsMax=<0 30 30>, angularLimitsMin=<0 -30 -30>, angularTarget=<0 0 0>}
      localJointOffset=<-5 0 0>
     }]
   }
   n14 = Anim Dynamics - Bone: lacesInner_r(ComponentPose=@n13, Alpha=1, ExternalForce="0,0,0") {
    angularDampingOverride=1
    angularSpringConstant=1500
    bAngularSpring=true
    bGravityOverrideInSimSpace=true
    bOverrideAngularDamping=true
    gravityScale=0.35
    simulationSpace=RootRelative
    alphaBoolBlend {blendOption=Linear}
    boundBone {boneName=lacesInner_r}
    physicsBodyDefinitions [{
      sphereCollisionRadius=0
      boundBone {boneName=lacesInner_r}
      boxExtents=<20 10 10>
      constraintSetup {angularLimitsMax=<0 30 30>, angularLimitsMin=<0 -30 -30>, angularTarget=<0 0 0>}
      localJointOffset=<7 0 0>
     }]
   }
   n15 = Anim Dynamics - Bone: lacesInner_l(ComponentPose=@n14, Alpha=1, ExternalForce="0,0,0") {
    angularDampingOverride=1
    angularSpringConstant=1500
    bAngularSpring=true
    bGravityOverrideInSimSpace=true
    bOverrideAngularDamping=true
    gravityScale=0.35
    simulationSpace=RootRelative
    alphaBoolBlend {blendOption=Linear}
    boundBone {boneName=lacesInner_l}
    physicsBodyDefinitions [{
      sphereCollisionRadius=0
      boundBone {boneName=lacesInner_l}
      boxExtents=<20 10 10>
      constraintSetup {angularLimitsMax=<0 30 30>, angularLimitsMin=<0 -30 -30>, angularTarget=<0 0 0>}
      localJointOffset=<6 0 0>
     }]
   }
   n16 = Anim Dynamics - Bone: lacesOut_l(ComponentPose=@n15, Alpha=1, ExternalForce="0,0,0") {
    angularDampingOverride=1
    angularSpringConstant=1500
    bAngularSpring=true
    bGravityOverrideInSimSpace=true
    bOverrideAngularDamping=true
    gravityScale=0.35
    simulationSpace=RootRelative
    alphaBoolBlend {blendOption=Linear}
    boundBone {boneName=lacesOut_l}
    physicsBodyDefinitions [{
      sphereCollisionRadius=0
      boundBone {boneName=lacesOut_l}
      boxExtents=<20 10 10>
      constraintSetup {angularLimitsMax=<0 30 30>, angularLimitsMin=<0 -30 -30>, angularTarget=<0 0 0>}
      localJointOffset=<-9 0 0>
     }]
   }
   n17 = Anim Dynamics - Bone: beltFlap(ComponentPose=@n16, Alpha=1, ExternalForce="0,0,0") {
    angularDampingOverride=1
    angularSpringConstant=1500
    bAngularSpring=true
    bGravityOverrideInSimSpace=true
    bOverrideAngularDamping=true
    gravityScale=0.35
    simulationSpace=RootRelative
    alphaBoolBlend {blendOption=Linear}
    boundBone {boneName=beltFlap}
    physicsBodyDefinitions [{
      sphereCollisionRadius=0
      boundBone {boneName=beltFlap}
      boxExtents=<10 20 20>
      constraintSetup {angularLimitsMax=<0 20 0>, angularLimitsMin=<0 -20 -40>, angularTarget=<0 0 0>}
      localJointOffset=<10 0 0>
     }]
   }
   n18 = Anim Dynamics - Bone: spine_05_bun_c(ComponentPose=@n17, Alpha=1, ExternalForce="0,0,0") {
    angularDampingOverride=1
    angularSpringConstant=1500
    bAngularSpring=true
    bGravityOverrideInSimSpace=true
    bOverrideAngularDamping=true
    gravityScale=0.35
    simulationSpace=RootRelative
    alphaBoolBlend {blendOption=Linear}
    alphaScaleBiasClamp {scale=0.05}
    boundBone {boneName=spine_05_bun_c}
    physicsBodyDefinitions [{
      sphereCollisionRadius=0
      boundBone {boneName=spine_05_bun_c}
      boxExtents=<10 20 20>
      constraintSetup {angularLimitsMax=<0 20 0>, angularLimitsMin=<0 -20 -40>, angularTarget=<0 0 0>}
      localJointOffset=<-20 0 0>
     }]
   }
   n19 = "headTop_bun_c.rotateZ = spine_05_bun_c.rotateZ - Bone Driven Controller"!(ComponentPose:Component Space Pose Link Structure=@n18, Alpha:float=1) -> (Pose:Component Space Pose Link Structure) {
    is=title AnimGraphNode_BoneDrivenController
    bAffectTargetRotationZ=true
    sourceComponent=RotationZ
    alphaBoolBlend {blendOption=Linear}
    sourceBone {boneName=spine_05_bun_c}
    targetBone {boneName=headTop_bun_c}
   }
   n20 = Anim Dynamics - Bone: relish_01(ComponentPose=@n19, Alpha=1, ExternalForce="0,0,0") {
    angularDampingOverride=1
    angularSpringConstant=1500
    bAngularSpring=true
    bGravityOverrideInSimSpace=true
    bOverrideAngularDamping=true
    gravityScale=0.35
    simulationSpace=RootRelative
    alphaBoolBlend {blendOption=Linear}
    alphaScaleBiasClamp {scale=0.6}
    boundBone {boneName=relish_01}
    physicsBodyDefinitions [{
      sphereCollisionRadius=0
      boundBone {boneName=relish_01}
      boxExtents=<10 20 20>
      constraintSetup {angularLimitsMax=<20 30 20>, angularLimitsMin=<-20 -30 -50>, angularTarget=<1 0 0>}
      localJointOffset=<10 0 0>
     }]
   }
   n21 = Anim Dynamics - Bone: relish_02(ComponentPose=@n20, Alpha=1, ExternalForce="0,0,0") {
    angularDampingOverride=1
    angularSpringConstant=1500
    bAngularSpring=true
    bGravityOverrideInSimSpace=true
    bOverrideAngularDamping=true
    gravityScale=0.35
    simulationSpace=RootRelative
    alphaBoolBlend {blendOption=Linear}
    alphaScaleBiasClamp {scale=0.6}
    boundBone {boneName=relish_02}
    physicsBodyDefinitions [{
      sphereCollisionRadius=0
      boundBone {boneName=relish_02}
      boxExtents=<10 20 20>
      constraintSetup {angularLimitsMax=<20 30 20>, angularLimitsMin=<-20 -30 -50>, angularTarget=<1 0 0>}
      localJointOffset=<7 0 0>
     }]
   }
   n22 = Anim Dynamics - Bone: relish_03(ComponentPose=@n21, Alpha=1, ExternalForce="0,0,0") {
    angularDampingOverride=1
    angularSpringConstant=1500
    bAngularSpring=true
    bGravityOverrideInSimSpace=true
    bOverrideAngularDamping=true
    gravityScale=0.35
    simulationSpace=RootRelative
    alphaBoolBlend {blendOption=Linear}
    alphaScaleBiasClamp {scale=0.6}
    boundBone {boneName=relish_03}
    physicsBodyDefinitions [{
      sphereCollisionRadius=0
      boundBone {boneName=relish_03}
      boxExtents=<10 20 20>
      constraintSetup {angularLimitsMax=<20 30 20>, angularLimitsMin=<-20 -30 -50>, angularTarget=<1 0 0>}
      localJointOffset=<12 0 0>
     }]
   }
   n23 = Anim Dynamics - Bone: relish_04(ComponentPose=@n22, Alpha=1, ExternalForce="0,0,0") {
    angularDampingOverride=1
    angularSpringConstant=1500
    bAngularSpring=true
    bGravityOverrideInSimSpace=true
    bOverrideAngularDamping=true
    gravityScale=0.35
    simulationSpace=RootRelative
    alphaBoolBlend {blendOption=Linear}
    alphaScaleBiasClamp {scale=0.6}
    boundBone {boneName=relish_04}
    physicsBodyDefinitions [{
      sphereCollisionRadius=0
      boundBone {boneName=relish_04}
      boxExtents=<10 20 20>
      constraintSetup {angularLimitsMax=<20 30 20>, angularLimitsMin=<-20 -30 -50>, angularTarget=<1 0 0>}
      localJointOffset=<9 0 0>
     }]
   }
   n24 = Anim Dynamics - Bone: relish_05(ComponentPose=@n23, Alpha=1, ExternalForce="0,0,0") {
    angularDampingOverride=1
    angularSpringConstant=1500
    bAngularSpring=true
    bGravityOverrideInSimSpace=true
    bOverrideAngularDamping=true
    gravityScale=0.35
    simulationSpace=RootRelative
    alphaBoolBlend {blendOption=Linear}
    alphaScaleBiasClamp {scale=0.6}
    boundBone {boneName=relish_05}
    physicsBodyDefinitions [{
      sphereCollisionRadius=0
      boundBone {boneName=relish_05}
      boxExtents=<10 20 20>
      constraintSetup {angularLimitsMax=<20 30 20>, angularLimitsMin=<-20 -30 -50>, angularTarget=<1 0 0>}
      localJointOffset=<11 0 0>
     }]
   }
   n25 = Anim Dynamics - Bone: relish_06(ComponentPose=@n24, Alpha=1, ExternalForce="0,0,0") {
    angularDampingOverride=1
    angularSpringConstant=1500
    bAngularSpring=true
    bGravityOverrideInSimSpace=true
    bOverrideAngularDamping=true
    gravityScale=0.35
    simulationSpace=RootRelative
    alphaBoolBlend {blendOption=Linear}
    alphaScaleBiasClamp {scale=0.6}
    boundBone {boneName=relish_06}
    physicsBodyDefinitions [{
      sphereCollisionRadius=0
      boundBone {boneName=relish_06}
      boxExtents=<10 20 20>
      constraintSetup {angularLimitsMax=<20 30 20>, angularLimitsMin=<-20 -30 -50>, angularTarget=<1 0 0>}
      localJointOffset=<10 0 0>
     }]
   }
   n26 = Anim Dynamics - Bone: relish_07(ComponentPose=@n25, Alpha=1, ExternalForce="0,0,0") {
    angularDampingOverride=1
    angularSpringConstant=1500
    bAngularSpring=true
    bGravityOverrideInSimSpace=true
    bOverrideAngularDamping=true
    gravityScale=0.35
    simulationSpace=RootRelative
    alphaBoolBlend {blendOption=Linear}
    alphaScaleBiasClamp {scale=0.6}
    boundBone {boneName=relish_07}
    physicsBodyDefinitions [{
      sphereCollisionRadius=0
      boundBone {boneName=relish_07}
      boxExtents=<10 20 20>
      constraintSetup {angularLimitsMax=<20 30 20>, angularLimitsMin=<-20 -30 -50>, angularTarget=<1 0 0>}
      localJointOffset=<13 0 0>
     }]
   }
   n27 = Anim Dynamics - Bone: relish_08(ComponentPose=@n26, Alpha=1, ExternalForce="0,0,0") {
    angularDampingOverride=1
    angularSpringConstant=1500
    bAngularSpring=true
    bGravityOverrideInSimSpace=true
    bOverrideAngularDamping=true
    gravityScale=0.35
    simulationSpace=RootRelative
    alphaBoolBlend {blendOption=Linear}
    alphaScaleBiasClamp {scale=0.6}
    boundBone {boneName=relish_08}
    physicsBodyDefinitions [{
      sphereCollisionRadius=0
      boundBone {boneName=relish_08}
      boxExtents=<10 20 20>
      constraintSetup {angularLimitsMax=<20 30 20>, angularLimitsMin=<-20 -30 -50>, angularTarget=<1 0 0>}
      localJointOffset=<8 0 0>
     }]
   }
   n28 = Anim Dynamics - Bone: relish_09(ComponentPose=@n27, Alpha=1, ExternalForce="0,0,0") {
    angularDampingOverride=1
    angularSpringConstant=1500
    bAngularSpring=true
    bGravityOverrideInSimSpace=true
    bOverrideAngularDamping=true
    gravityScale=0.35
    simulationSpace=RootRelative
    alphaBoolBlend {blendOption=Linear}
    alphaScaleBiasClamp {scale=0.6}
    boundBone {boneName=relish_09}
    physicsBodyDefinitions [{
      sphereCollisionRadius=0
      boundBone {boneName=relish_09}
      boxExtents=<10 20 20>
      constraintSetup {angularLimitsMax=<20 30 20>, angularLimitsMin=<-20 -30 -50>, angularTarget=<1 0 0>}
      localJointOffset=<11 0 0>
     }]
   }
   n29 = Get TargetingAlpha()
   n30 = "Transform (Modify) Bone - Bone: spine_04_bun_r"(ComponentPose=@n28, Alpha=@n29, Translation="0,4,6", Rotation="0,0,0", Scale="1,1,1") {
    translationMode=BMM_Additive
    translationSpace=BCS_ParentBoneSpace
    alphaBoolBlend {blendOption=Linear}
    alphaScaleBiasClamp {bInterpResult=true, interpSpeedDecreasing=5, interpSpeedIncreasing=15}
    boneToModify {boneName=spine_04_bun_r}
    translation {y=4, z=6}
   }
   n31 = "Transform (Modify) Bone - Bone: clavicle_aim_r"(ComponentPose=@n30, Alpha=@n29, Translation="1.5,0.000,0", Rotation="0,0,0", Scale="1,1,1") {
    translationMode=BMM_Additive
    translationSpace=BCS_ParentBoneSpace
    alphaBoolBlend {blendOption=Linear}
    alphaScaleBiasClamp {bInterpResult=true, interpSpeedDecreasing=5, interpSpeedIncreasing=15}
    boneToModify {boneName=clavicle_aim_r}
    translation {x=1.5}
   }
   Output Pose(Result=Component To Local(ComponentPose=@n31)) {layerGroup=None}
  ]}
 EventGraph {links [], nodes []}
]
variables []
